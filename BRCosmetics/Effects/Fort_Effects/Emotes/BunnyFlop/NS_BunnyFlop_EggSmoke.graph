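# Niagara system for the BunnyFlop emote, producing a brief egg smoke puff with a ring, placed using a system-level OffsetPosition.
# Smoke bursts four sprites that drift outward and slightly up under drag, growing with SubUV animation and a soft alpha fade-in and fade-out.
# SnapRing bursts one short-lived sprite with a Z locked axis, expanding from zero as it fades.
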NiagaraSystem /BRCosmetics/Effects/Fort_Effects/Emotes/BunnyFlop/NS_BunnyFlop_EggSmoke
dependencies {
 usedDataInterfaces [&NiagaraDataInterfaceCurve, &NiagaraDataInterfaceColorCurve]
 usedDynamicInputs []
 usedModules [
  &SystemState
  &self:SystemSpawnScript.SystemScriptSource.SystemScriptGraph.NiagaraNodeAssignment_0.SetVariables_#1
 ]
}
emitters {
 Smoke {
  emitterUpdateScript [
   &EmitterState(Life Cycle Mode=::NewEnumerator1, Loop Behavior=::NewEnumerator1, Loop Duration=1, UseLoopDelay=false)
   &SpawnBurst_Instantaneous(Spawn Count=4, Use Spawn Probability=false)
  ]
  particleSpawnScript [
   /Niagara/Modules/Spawn/Initialization/InitializeParticle(
    Lifetime=&UniformRangedFloat(Minimum=0.8, Maximum=0.9)
    Mesh Scale=<1.5 1.5 1.5>
    Module.Write Color=false
    Module.Write Lifetime=true
    Module.Write Mass=true
    Module.Write Material Random=true
    Module.Write Position=false
    Module.Write Scale=false
    Module.Write SpriteRotation=true
    Module.Write SpriteSize=true
    Position=&SimulationPosition()
    Sprite Rotation=&UniformRangedFloat(Minimum=0, Maximum=1)
    Sprite Size=&UniformRangedVector2D(
     Minimum=&Vector2DFromFloat(Value=40)
     Maximum=&Vector2DFromFloat(Value=50)
    )
   )
   /Niagara/Modules/Spawn/Location/SphereLocation(
    MaskSpawnGroup=false
    Module.Override Randomness=false
    Sphere Origin:NiagaraPosition=&Add_Vector(A=Linked() {linkedVariable {name=Particles.Position, type=NiagaraPosition}}, B=$System.OffsetPosition)
    Sphere Radius=0
    Surface Only=false
   )
   &AddVelocity(Velocity=&UniformRangedVector(Minimum=<-20 -20 5>, Maximum=<20 20 10>))
  ]
  particleUpdateScript [
   &ParticleState()
   &Drag(Drag=3, Use Linear Drag=true, Use Rotational Drag=false)
   &SolveForcesAndVelocity(Clamp Velocity=false, Limit Acceleration=false)
   &Color(Color=&ColorFromCurve(ColorCurve=DataInterface() {propertyValues {
       AlphaCurve {keys [<at -0.0023 0 RCIM_Cubic>,<at 0.5 0.15>,<at 0.9977 -0.003 RCIM_Cubic>]}
       BlueCurve {keys [<at -0.0023 0.7758 RCIM_Cubic>,<at 0.9977 0.7758 RCIM_Cubic>]}
       GreenCurve {keys [<at -0.0023 0.7758 RCIM_Cubic>,<at 0.9977 0.7758 RCIM_Cubic>]}
       RedCurve {keys [<at -0.0023 0.7758 RCIM_Cubic>,<at 0.9977 0.7758 RCIM_Cubic>]}
      }}))
   /Niagara/Modules/Update/SubUV/SubUVAnimation(Number Of Frames=20)
   &ScaleSpriteSize(Scale Factor=&Vector2DFromFloat(Value=&FloatFromCurve(FloatCurve=DataInterface() {propertyValues {Curve {keys [<at 0 0.5 RCIM_Cubic>,<at 1 1 RCIM_Cubic>]}}})))
  ]
  rendererClasses [&NiagaraSpriteRendererProperties]
  renderers [{rendererClass=&NiagaraSpriteRendererProperties}]
 }
 SnapRing {
  emitterUpdateScript [
   &EmitterState(Life Cycle Mode=::NewEnumerator1, Loop Behavior=::NewEnumerator1, Loop Duration=1, UseLoopDelay=false)
   &SpawnBurst_Instantaneous(Spawn Count=1, Use Spawn Probability=false)
  ]
  particleSpawnScript [
   /SetVariables(Particles.LockedAxis:Vector3f=<0 0 1>) {isSetParameters=true}
   &InitializeParticle(
    Lifetime=&UniformRangedFloat(Minimum=0.2, Maximum=0.4)
    Mesh Scale=<1.5 1.5 1.5>
    Module.Write Color=false
    Module.Write Lifetime=true
    Module.Write Mass=true
    Module.Write Material Random=true
    Module.Write Position=false
    Module.Write Scale=false
    Module.Write SpriteRotation=false
    Module.Write SpriteSize=true
    Position=&SimulationPosition()
    Sprite Size=<150 150>
   )
   &SphereLocation(
    MaskSpawnGroup=false
    Module.Override Randomness=false
    Offset=<0 0 -2>
    Sphere Origin:NiagaraPosition=&Add_Vector(A=Linked() {linkedVariable {name=Particles.Position, type=NiagaraPosition}}, B=$System.OffsetPosition)
    Sphere Radius=0
    Surface Only=false
   )
  ]
  particleUpdateScript [
   &ParticleState()
   &SolveForcesAndVelocity(Clamp Velocity=false, Limit Acceleration=false)
   &ScaleSpriteSize(Scale Factor=&Vector2DFromFloat(Value=&FloatFromCurve(FloatCurve=DataInterface() {propertyValues {Curve {keys [<at 0 0>,<at 1 1>]}}})))
   &Color(Color=&ColorFromCurve(ColorCurve=DataInterface() {propertyValues {
       AlphaCurve {keys [<at -0.0023 0.5 RCIM_Cubic>,<at 0.9977 -0.003 RCIM_Cubic>]}
       BlueCurve {keys [<at -0.0023 0.5879 RCIM_Cubic>,<at 0.9977 0.5879 RCIM_Cubic>]}
       GreenCurve {keys [<at -0.0023 0.6667 RCIM_Cubic>,<at 0.9977 0.6667 RCIM_Cubic>]}
       RedCurve {keys [<at -0.0023 0.7758 RCIM_Cubic>,<at 0.9977 0.7758 RCIM_Cubic>]}
      }}))
  ]
  rendererClasses [&NiagaraSpriteRendererProperties]
  renderers [{rendererClass=&NiagaraSpriteRendererProperties}]
 }
}
summary {
 systemName=NS_BunnyFlop_EggSmoke
 emitters [
  {emitterName=SnapRing, rendererClasses [&NiagaraSpriteRendererProperties]}
  {emitterName=Smoke, rendererClasses [&NiagaraSpriteRendererProperties]}
 ]
 userVariables []
}
